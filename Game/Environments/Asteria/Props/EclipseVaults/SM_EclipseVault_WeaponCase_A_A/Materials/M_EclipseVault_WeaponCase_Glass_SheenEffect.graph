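# Translucent Default Lit glass material for the Eclipse Vault weapon case, building an animated, view-dependent sheen.
# A glint derived from UVs and camera vector is sharpened, weighted toward facing surfaces and faded in with camera distance.
# A thin scrolling band, from rotated XZ position via Panner, Frac and Floor, gets a cubic falloff, inverted-Fresnel mask and SweepColor tint.
# Both layers are scaled and summed.
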Material /Game/Environments/Asteria/Props/EclipseVaults/SM_EclipseVault_WeaponCase_A_A/Materials/M_EclipseVault_WeaponCase_Glass_SheenEffect
graph {
 properties {blendMode=BLEND_Translucent, pixelDepthOffsetMode=PDOM_Legacy, usedShadingModels=Default Lit}
 nodes [
  n1 = Saturate(None=DotProduct(A=VertexNormalWS(), B=CameraVectorWS()))
  n2 = TextureCoordinate()
  n3 = MaterialFunctionCall(Float3=Normalize(VectorInput=CameraVectorWS())) {materialFunction=&BreakOutFloat3Components}
  n4 = Add(A=Add(A=ComponentMask(None=@n2) {r=true}, B=ComponentMask(None=@n2) {g=true}), B=Add(A=Add(A=@n3.R, B=@n3.R), B=@n3.B)) {constB=0}
  n5 = Multiply(A=Abs(None=@n4), B=OneMinus(None=@n4))
  n6 = Power(Base=Clamp(None=Multiply(A=@n5) {constB=4}), Exp=ScalarParameter() {defaultValue=20, parameterName=Sweep Exponent})
  n7 = ScalarParameter() {defaultValue=10, parameterName=Sweep Brightness}
  n8 = Saturate(None=Multiply(A=Power(Base=@n1) {constExponent=5}, B=Multiply(A=@n6, B=@n7)))
  n9 = Divide(A=Distance(A=WorldPosition().XYZ, B=CameraPositionWS())) {constB=1000}
  n10 = Subtract(
   A=MaterialFunctionCall() {materialFunction=&LocalPosition}.Local Position
   B=ObjectPositionWS()
  )
  n11 = Multiply(
   A=MaterialFunctionCall(UVs=ComponentMask(None=@n10) {b=true, r=true}, "Rotation Angle (0-1)"=Constant() {r=0.1}) {materialFunction=&CustomRotator}.Rotated Values
   B=ScalarParameter() {defaultValue=0.001, parameterName=WorldTileScale}
  )
  n12 = ScalarParameter() {defaultValue=0.3, parameterName=Scroll Speed}
  n13 = Frac(None=ComponentMask(None=Panner(Coordinate=@n11, Speed=@n12) {speedX=0.2, speedY=0.2}) {g=true})
  n14 = Floor(None=Add(A=@n13) {constB=0.1})
  n15 = Subtract(A=@n13) {constB=0.9}
  n16 = Power(Base=OneMinus(None=Multiply(A=@n15) {constB=10})) {constExponent=3}
  n17 = Add(A=MaterialFunctionCall() {materialFunction=&MF_Fresnel_FX}.Fresnel Inverted) {constB=0.1}
  n18 = Multiply(
   A=Multiply(A=Multiply(A=@n14, B=@n16), B=Saturate(None=@n17))
   B=VectorParameter() {parameterName=SweepColor, defaultValue=<rgba 1.8958 2 1.9078 0>}.RGB
  )
  Add(
   A=Multiply(A=Multiply(A=@n8) {constB=0.02}, B=Saturate(None=@n9))
   B=Multiply(A=Saturate(None=@n18), B=ScalarParameter() {defaultValue=0.25, parameterName=Sheen Intensity}) {constB=0.1}
  )
  Constant() {r=1}
  VectorParameter() {parameterName=Emissive, defaultValue=<rgba 0.0019 0.0025 0.0052 1>}
  WorldPosition()
 ]
}
